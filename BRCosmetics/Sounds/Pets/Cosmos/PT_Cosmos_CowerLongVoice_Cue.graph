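# Sound cue for the Cosmos pet's long cower vocalization, using close-range pet attenuation and the creature pets sound class.
# An outer random node weights an empty branch twice as heavily as the voice branch, so roughly one trigger in three makes sound.
# The voice branch picks evenly among three Cosmos vocal waves.
# A one-maximum-per-owner, prevent-new concurrency rule stops overlapping plays.
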SoundCue /BRCosmetics/Sounds/Pets/Cosmos/PT_Cosmos_CowerLongVoice_Cue
root=Random_0
nodes [
 {id=Random_0, children [Random_3,null], properties {bRandomizeWithoutReplacement=false, weights [0.5,1]}}
 {id=Random_3, children [WavePlayer_0,WavePlayer_2,WavePlayer_3], properties {weights [1,1,1]}}
 {id=WavePlayer_0, properties {soundWaveAssetPtr=&/Emote_Cosmos_Pram_Close_A_Vocal_02}}
 {id=WavePlayer_2, properties {soundWaveAssetPtr=&/Emote_Cosmos_Pram_Close_A_Vocal_03}}
 {id=WavePlayer_3, properties {soundWaveAssetPtr=&/Emote_Cosmos_Pram_Close_A_Vocal_06}}
]
properties {
 attenuationSettings=&Fort_Pets_Close_Athena_Att
 maxDistance=1800.0001
 soundClassObject=&Fort_Creatures_Pets
 volumeMultiplier=0.448
 concurrencySet [/FNGameplayCues/Sounds/Concurrency_Rules/OneMaximum_Owner_PreventNew_Succes_Cues]
}
