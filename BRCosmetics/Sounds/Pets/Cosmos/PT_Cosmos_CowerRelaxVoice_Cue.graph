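# Sound cue for the Cosmos pet's cower-relax voice, drawing on five relax voice recordings.
# After a short fixed delay, a weighted random node picks either the voice pool or an empty branch, so roughly two in three triggers are silent.
# The pool picks among equally weighted waves.
# It uses the creature pets sound class, close-range pet attenuation, and a concurrency rule named for one sound per owner, preventing new ones.
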SoundCue /BRCosmetics/Sounds/Pets/Cosmos/PT_Cosmos_CowerRelaxVoice_Cue
root=Delay_0
nodes [
 {id=Delay_0, children [Random_0], properties {delayMax=0.2, delayMin=0.2}}
 {id=Random_0, children [Random_2,null], properties {bRandomizeWithoutReplacement=false, weights [0.5,1]}}
 {id=Random_2, children [WavePlayer_3, WavePlayer_6, WavePlayer_9, WavePlayer_10, WavePlayer_11], properties {weights [1,1,1,1,1]}}
 {id=WavePlayer_10, properties {soundWaveAssetPtr=&/PT_Cosmos_RelaxVoice_07}}
 {id=WavePlayer_11, properties {soundWaveAssetPtr=&/PT_Cosmos_RelaxVoice_08}}
 {id=WavePlayer_3, properties {soundWaveAssetPtr=&/PT_Cosmos_RelaxVoice_10}}
 {id=WavePlayer_6, properties {soundWaveAssetPtr=&/PT_Cosmos_RelaxVoice_03}}
 {id=WavePlayer_9, properties {soundWaveAssetPtr=&/PT_Cosmos_RelaxVoice_06}}
]
properties {
 attenuationSettings=&Fort_Pets_Close_Athena_Att
 maxDistance=1800.0001
 soundClassObject=&Fort_Creatures_Pets
 volumeMultiplier=0.448
 concurrencySet [/FNGameplayCues/Sounds/Concurrency_Rules/OneMaximum_Owner_PreventNew_Succes_Cues]
}
