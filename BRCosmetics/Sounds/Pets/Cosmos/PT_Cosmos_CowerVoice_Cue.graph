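# Sound cue for the cower vocalization of the Cosmos pet.
# An outer random node, with replacement, picks between silence and the voice branch, so roughly one trigger in three plays a sound.
# The voice branch randomly picks one of four equally weighted waves without replacement.
# Playback uses close-range pet attenuation, the creature pets sound class, and a concurrency set named for one sound per owner, preventing new ones.
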SoundCue /BRCosmetics/Sounds/Pets/Cosmos/PT_Cosmos_CowerVoice_Cue
root=Random_2
nodes [
 {id=Random_1, children [WavePlayer_4,WavePlayer_2,WavePlayer_1,WavePlayer_3], properties {weights [1,1,1,1]}}
 {id=Random_2, children [Random_1,null], properties {bRandomizeWithoutReplacement=false, weights [0.5,1]}}
 {id=WavePlayer_1, properties {soundWaveAssetPtr=&/PT_Cosmos_CowerVoice_05}}
 {id=WavePlayer_2, properties {soundWaveAssetPtr=&/PT_Cosmos_CowerVoice_04}}
 {id=WavePlayer_3, properties {soundWaveAssetPtr=&/PT_Cosmos_CowerVoice_06}}
 {id=WavePlayer_4, properties {soundWaveAssetPtr=&/PT_Cosmos_CowerVoice_01}}
]
properties {
 attenuationSettings=&Fort_Pets_Close_Athena_Att
 maxDistance=1800.0001
 soundClassObject=&Fort_Creatures_Pets
 volumeMultiplier=0.672
 concurrencySet [/FNGameplayCues/Sounds/Concurrency_Rules/OneMaximum_Owner_PreventNew_Succes_Cues]
}
