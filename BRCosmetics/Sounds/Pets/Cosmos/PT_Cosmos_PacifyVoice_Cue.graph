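# Sound cue for the Cosmos pet's pacify voice in the BR cosmetics pet set.
# A random node picks between two equally weighted wave variations.
# Close-range pet attenuation and the creature pets sound class shape playback at reduced volume.
# A per-owner concurrency rule allows one instance at a time, stopping the oldest.
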SoundCue /BRCosmetics/Sounds/Pets/Cosmos/PT_Cosmos_PacifyVoice_Cue
root=Random_0
nodes [
 {id=Random_0, children [WavePlayer_0,WavePlayer_1], properties {weights [1,1]}}
 {id=WavePlayer_0, properties {soundWaveAssetPtr=&/PT_Cosmos_PacifyVoice_01}}
 {id=WavePlayer_1, properties {soundWaveAssetPtr=&/PT_Cosmos_PacifyVoice_02}}
]
properties {
 attenuationSettings=&Fort_Pets_Close_Athena_Att
 maxDistance=1800.0001
 soundClassObject=&Fort_Creatures_Pets
 volumeMultiplier=0.672
 concurrencySet [/Game/Sounds/Concurrency_Rules/OneMaximum_Owner_StopOldest]
}
