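# Sound cue for the Cosmos pet's foley pram grab effect, named as part of a stand-up animation sound.
# A single wave player feeds the output directly with one grab foley wave, with no randomization or modulation.
# It uses close-range pet attenuation, the creature pets sound class, a slight volume boost, and a concurrency rule that keeps one instance per owner by stopping the oldest.
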SoundCue /BRCosmetics/Sounds/Pets/Cosmos/PT_Cosmos_PlaySFX_FoleyPramGrab
root=WavePlayer_0
nodes [{id=WavePlayer_0, properties {soundWaveAssetPtr=&/Pet_Cosmos_Stand_Up_GrabFoley_Pram_01}}]
properties {
 attenuationSettings=&Fort_Pets_Close_Athena_Att
 maxDistance=1800.0001
 soundClassObject=&Fort_Creatures_Pets
 volumeMultiplier=1.12
 concurrencySet [/Game/Sounds/Concurrency_Rules/OneMaximum_Owner_StopOldest]
}
