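# Sound cue for the Cosmos pet's sip sound effect.
# A random node picks one of three sip wave variations with equal weight.
# It plays through the pets sound class with close-range pet attenuation and a slight volume boost.
# Concurrency allows one instance per owner, stopping the oldest when a new one starts.
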SoundCue /BRCosmetics/Sounds/Pets/Cosmos/PT_Cosmos_SipSFX_Cue
root=Random_1
nodes [
 {id=Random_1, children [WavePlayer_3,WavePlayer_4,WavePlayer_7], properties {weights [1,1,1]}}
 {id=WavePlayer_3, properties {soundWaveAssetPtr=&/PT_Cosmos_SipSFX_01}}
 {id=WavePlayer_4, properties {soundWaveAssetPtr=&/PT_Cosmos_SipSFX_02}}
 {id=WavePlayer_7, properties {soundWaveAssetPtr=&/PT_Cosmos_SipSFX_03}}
]
properties {
 attenuationSettings=&Fort_Pets_Close_Athena_Att
 maxDistance=1800.0001
 soundClassObject=&Fort_Creatures_Pets
 volumeMultiplier=1.12
 concurrencySet [/Game/Sounds/Concurrency_Rules/OneMaximum_Owner_StopOldest]
}
