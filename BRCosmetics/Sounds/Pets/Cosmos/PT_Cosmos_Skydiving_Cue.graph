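# Sound cue for the Cosmos pet's skydiving vocalization, routed to the creature pets sound class with close-range pet attenuation.
# An outer random node plays a voice line about one time in three and is silent otherwise through an empty, higher-weighted branch.
# The inner random node picks among four equally weighted skydive voice waves.
# Concurrency allows one instance per owner and blocks new plays while one is active.
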SoundCue /BRCosmetics/Sounds/Pets/Cosmos/PT_Cosmos_Skydiving_Cue
root=Random_1
nodes [
 {id=Random_0, children [WavePlayer_0,WavePlayer_1,WavePlayer_3,WavePlayer_5], properties {weights [1,1,1,1]}}
 {id=Random_1, children [Random_0,null], properties {bRandomizeWithoutReplacement=false, weights [0.5,1]}}
 {id=WavePlayer_0, properties {soundWaveAssetPtr=&/PT_Cosmos_SkyDiveVoice_13}}
 {id=WavePlayer_1, properties {soundWaveAssetPtr=&/PT_Cosmos_SkyDiveVoice_15}}
 {id=WavePlayer_3, properties {soundWaveAssetPtr=&/PT_Cosmos_SkyDiveVoice_03}}
 {id=WavePlayer_5, properties {soundWaveAssetPtr=&/PT_Cosmos_SkyDiveVoice_09}}
]
properties {
 attenuationSettings=&Fort_Pets_Close_Athena_Att
 bOverrideConcurrency=true
 maxDistance=1800.0001
 soundClassObject=&Fort_Creatures_Pets
 volumeMultiplier=0.672
 concurrencyOverrides {bLimitToOwner=true, maxCount=1, resolutionRule=PreventNew}
}
