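# Sound cue for the Cosmos pet's stand-up vocalization, using pet close-range attenuation and the creature pets sound class.
# An outer random node, drawing with replacement, picks between an empty branch and the voice branch, so it stays silent about two times in three.
# The voice branch randomly selects one of three equally weighted stand-up waves.
# A concurrency rule allows one instance per owner, stopping the oldest.
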SoundCue /BRCosmetics/Sounds/Pets/Cosmos/PT_Cosmos_StandVoice
root=Random_2
nodes [
 {id=Random_0, children [WavePlayer_0,WavePlayer_1,WavePlayer_2], properties {weights [1,1,1]}}
 {id=Random_2, children [Random_0,null], properties {bRandomizeWithoutReplacement=false, weights [0.5,1]}}
 {id=WavePlayer_0, properties {soundWaveAssetPtr=&/PT_Cosmos_Stand_Up_Voice_01}}
 {id=WavePlayer_1, properties {soundWaveAssetPtr=&/PT_Cosmos_Stand_Up_Voice_02}}
 {id=WavePlayer_2, properties {soundWaveAssetPtr=&/PT_Cosmos_Stand_Up_Voice_03}}
]
properties {
 attenuationSettings=&Fort_Pets_Close_Athena_Att
 maxDistance=1800.0001
 soundClassObject=&Fort_Creatures_Pets
 volumeMultiplier=0.56
 concurrencySet [/Game/Sounds/Concurrency_Rules/OneMaximum_Owner_StopOldest]
}
